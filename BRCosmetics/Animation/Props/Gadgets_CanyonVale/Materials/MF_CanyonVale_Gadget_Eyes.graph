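# Material function that layers material-driven eyes with steerable gaze onto incoming material attributes for the CanyonVale gadget emote prop.
# Horizontal and vertical look parameters remap from -1..1 into min/max UV offsets, added to size-scaled UVs that sample clamped eye color, normal and packed specular/roughness/metallic textures.
# A fixed UV-space sphere mask, EyeMask, blends these over the base attributes, so the gaze moves inside a stationary region.
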MaterialFunction /BRCosmetics/Animation/Props/Gadgets_CanyonVale/Materials/MF_CanyonVale_Gadget_Eyes
graph {nodes [
  n1 = GetMaterialAttributes(None=FunctionInput() {inputType=FunctionInput_MaterialAttributes}) {attributeGetTypes [
    #1
    #2
    #3
    #4
    #5
   ]}
  n2 = ScalarParameter() {group=Material Driven Eyes, parameterName=min_lookU, sortPriority=34}
  n3 = ScalarParameter() {defaultValue=0.157, group=Material Driven Eyes, parameterName=max_lookU, sortPriority=34}
  n4 = Multiply(A=Add(A=ScalarParameter() {group=Material Driven Eyes, parameterName=look_horizontal, sliderMax=1, sliderMin=-1})) {constB=0.5}
  n5 = ScalarParameter() {defaultValue=0.085, group=Material Driven Eyes, parameterName=min_lookV, sortPriority=33}
  n6 = ScalarParameter() {defaultValue=-0.155, group=Material Driven Eyes, parameterName=max_lookV, sortPriority=33}
  n7 = Multiply(A=Add(A=ScalarParameter() {group=Material Driven Eyes, parameterName=look_vertical, sliderMax=1, sliderMin=-1})) {constB=0.5}
  n8 = Add(
   A=Multiply(A=TextureCoordinate(), B=ScalarParameter() {defaultValue=1, group=Material Driven Eyes, parameterName=Size})
   B=AppendVector(A=LinearInterpolate(A=@n2, B=@n3, Alpha=@n4), B=LinearInterpolate(A=@n5, B=@n6, Alpha=@n7))
  )
  n9 = TextureSample(UVs=@n8) {samplerSource=SSM_Clamp_WorldGroupSettings, texture=/BRCosmetics/Animation/Props/Gadgets_CanyonVale/Textures/T_CanyonVale_Eyes_EmoteProp_D}
  n10 = TextureSample(UVs=@n8) {
   samplerSource=SSM_Clamp_WorldGroupSettings
   samplerType=SAMPLERTYPE_Normal
   texture=/BRCosmetics/Animation/Props/Gadgets_CanyonVale/Textures/T_CanyonVale_Eyes_EmoteProp_N
  }
  n11 = TextureSample(UVs=@n8) {samplerSource=SSM_Clamp_WorldGroupSettings, texture=/BRCosmetics/Animation/Props/Gadgets_CanyonVale/Textures/T_CanyonVale_Eyes_EmoteProp_S}
  n12 = SetMaterialAttributes!(
   MaterialAttributes=@n1.MaterialAttributes
   BaseColor=LinearInterpolate(A=@n1.BaseColor, B=@n9.RGB, Alpha=NamedRerouteUsage())
   Normal=LinearInterpolate(A=@n1.Normal, B=@n10.RGB, Alpha=NamedRerouteUsage())
   Specular=LinearInterpolate(A=@n1.Specular, B=@n11.R, Alpha=NamedRerouteUsage())
   Roughness=LinearInterpolate(A=@n1.Roughness, B=@n11.G, Alpha=NamedRerouteUsage())
   Metallic=LinearInterpolate(A=@n1.Metallic, B=@n11.B, Alpha=NamedRerouteUsage())
  ) {attributeSetTypes [
    #1
    #2
    #3
    #4
    #5
   ]}
  FunctionOutput(None=@n12)
  n13 = TextureCoordinate() {uTiling=1.1}
  n14 = Constant2Vector() {g=0.21, r=0.25}
  NamedRerouteDeclaration(None=SphereMask(A=Add(A=@n13, B=@n14), B=Constant() {r=0.5}, Radius=Constant() {r=0.2344}) {attenuationRadius=0.15}) {name=EyeMask, nodeColor=<rgba 1 0.1641 0 1>}
  TextureSample() {samplerType=SAMPLERTYPE_Normal, texture=/BRCosmetics/Animation/Props/Gadgets_CanyonVale/Textures/T_CanyonVale_EmoteProp_N}
 ]}
