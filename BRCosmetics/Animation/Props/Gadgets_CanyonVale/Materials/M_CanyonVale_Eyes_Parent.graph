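# Parent material for the eyes of the CanyonVale gadget cosmetic props, built for skeletal meshes with morph targets.
# It uses material attributes, with the shading model driven by the material expression.
# The shared 2023 character-parent PreFX function feeds a CanyonVale gadget eyes function, whose result enters the matching PostFX function.
# PreFX also passes auxiliary data directly to PostFX, so the eye logic sits between the standard character effect stages.
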Material /BRCosmetics/Animation/Props/Gadgets_CanyonVale/Materials/M_CanyonVale_Eyes_Parent
graph {
 properties {
  bUseMaterialAttributes=true
  bUsedWithMorphTargets=true
  bUsedWithSkeletalMesh=true
  pixelDepthOffsetMode=PDOM_Legacy
  shadingModel=MSM_FromMaterialExpression
  usedShadingModels=Default Lit | Subsurface | Clear Coat | Subsurface Profile | Cloth
 }
 nodes [
  n1 = MaterialFunctionCall() {materialFunction=&MF_CharacterParent_2023_PreFX}
  MaterialFunctionCall(
   Material=MaterialFunctionCall!(In=GetMaterialAttributes(None=@n1.Result).MaterialAttributes) {materialFunction=&/MF_CanyonVale_Gadget_Eyes}.Result
   M=@n1.M
   Pre Skinned Local Position=@n1.M
   Distance Blend=@n1.M
   Fresnel=@n1.M
  ) {materialFunction=&MF_CharacterParent_2023_PostFX}
 ]
}
